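# Post-process-style AnimBP for the SunTide Upgrade male body cosmetic, adding secondary motion to belt, can, shorts and tassels.
# It copies the attached parent's pose through deform and corrective Control Rigs, fed by emote and crouch states from EventGraph.
# Bone constraints and pelvis/calf-space RigidBody sims drive dangling parts and shorts, while AnimDynamics plus planar-limited Trail controllers swing left-forearm tassels.
# EventGraph also computes a gravity override, which AnimGraph never reads.
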AnimBlueprint /BRCosmetics/Characters/Player/Male/Medium/Bodies/M_MED_SunTide_Upgrade/Meshes/M_MED_SunTide_Upgrade_AnimBP
target_skeleton=&/M_MED_SunTide_Upgrade_Skeleton
graphs [
 AnimGraph {nodes [
   "Constraint - dyn_belt_mstr (PPP)"!(ComponentPose, Alpha=1, ConstraintWeights_0=0.5658, ConstraintWeights_1:float=0.2324, ConstraintWeights_2:float=0.2018) -> (Pose) {
    alphaBoolBlend {blendOption=Linear}
    boneToModify {boneName=dyn_belt_mstr}
    constraintSetup [{targetBone {boneName=pelvis}}, {targetBone {boneName=thigh_twist_01_l}}, {targetBone {boneName=deform_groin_l}}]
    constraintWeights [1,1,1]
   }
   "Transform (Modify) Bone - Bone: dyn_belt_mstr"!(ComponentPose, Alpha=1, Translation="2,1,2", Rotation="15,0,-20") -> (Pose) {
    rotationMode=BMM_Replace
    rotationSpace=BCS_BoneSpace
    scaleMode=BMM_Replace
    scaleSpace=BCS_BoneSpace
    translationMode=BMM_Additive
    translationSpace=BCS_BoneSpace
    alphaBoolBlend {blendOption=Linear}
    alphaScaleBiasClamp {bInterpResult=true, interpSpeedDecreasing=20, interpSpeedIncreasing=7}
    boneToModify {boneName=dyn_belt_mstr}
    rotation {pitch=15, roll=-20}
    translation=<2 1 2>
   }
   "Transform (Modify) Bone - Bone: dyn_belt_mstr"(Alpha=1, Translation="2,1,2", Rotation="15,0,-20") {
    rotationMode=BMM_Replace
    rotationSpace=BCS_BoneSpace
    scaleMode=BMM_Replace
    scaleSpace=BCS_BoneSpace
    translationMode=BMM_Additive
    translationSpace=BCS_BoneSpace
    alphaBoolBlend {blendOption=Linear}
    alphaScaleBiasClamp {bInterpResult=true, interpSpeedDecreasing=20, interpSpeedIncreasing=7}
    boneToModify {boneName=dyn_belt_mstr}
    rotation {pitch=15, roll=-20}
    translation=<2 1 2>
   }
   n1 = Control Rig(Source=Copy Pose From Mesh() {bUseAttachedParent=true}) {
    alphaBoolBlend {blendOption=Linear}
    controlRigAssetReference {blueprintRigClass=&/M_MED_SunTide_Upgrade_DeformRig.M_MED_SunTide_Upgrade_DeformRig_C}
   }
   n2 = Control Rig!(Source=@n1, Alpha, Emote:double=Get Emote() -> (Emote:double)) -> (Pose) {
    alphaBoolBlend {blendOption=Linear}
    controlRigAssetReference {blueprintRigClass=&/M_MED_SunTide_Upgrade_CtrlRig3.M_MED_SunTide_Upgrade_CtrlRig3_C}
   }
   n3 = "Constraint - dyn_can_mstr (PP)"!(ComponentPose=Local To Component(LocalPose=@n2), Alpha=1, ConstraintWeights_0=0.67, ConstraintWeights_1:float=0.33) -> (Pose) {
    alphaBoolBlend {blendOption=Linear}
    boneToModify {boneName=dyn_can_mstr}
    constraintSetup [{targetBone {boneName=pelvis}}, {targetBone {boneName=thigh_twist_01_r}}]
    constraintWeights [1,1]
   }
   n4 = "Transform (Modify) Bone - Bone: dyn_can_mstr"(ComponentPose=@n3, Alpha=0, Translation="2.512,-5.316,-0.811", Rotation="-16.350,33.227,-23.146", Scale="1,1,1") {
    rotationMode=BMM_Additive
    translationMode=BMM_Additive
    alphaBoolBlend {blendOption=Linear}
    boneToModify {boneName=dyn_can_mstr}
    rotation=<rot -16.35 33.227 -23.146>
    translation=<2.512 -5.316 -0.811>
   }
   n5 = RigidBody(ComponentPose=@n4) {
    bOverrideWorldGravity=true
    overridePhysicsAsset=&/M_MED_SunTide_Upgrade_Physics
    simulationSpace=BaseBoneSpace
    alphaBoolBlend {blendOption=Linear}
    baseBoneRef {boneName=pelvis}
    componentAppliedLinearAccClamp=<3000 3000 3000>
    componentLinearAccScale=<0.3 0.3 0.5>
    componentLinearVelScale=<0.3 0.3 0.5>
    overrideWorldGravity {z=-980}
   }
   n6 = "Constraint - dyn_belt_02 (PPP)"!(ComponentPose=@n5, Alpha=1, ConstraintWeights_0=0.2603, ConstraintWeights_1:float=0.5555, ConstraintWeights_2:float=0.1842) -> (Pose) {
    alphaBoolBlend {blendOption=Linear}
    alphaScaleBiasClamp {bClampResult=true, bInterpResult=true, clampMax=0.45, interpSpeedDecreasing=20, interpSpeedIncreasing=20}
    boneToModify {boneName=dyn_belt_02}
    constraintSetup [{targetBone {boneName=pelvis}}, {targetBone {boneName=thigh_twist_01_l}}, {targetBone {boneName=deform_groin_l}}]
    constraintWeights [1,1,1]
   }
   n7 = "Constraint - dyn_belt_03 (PPPPP)"!(
    ComponentPose=@n6
    Alpha=1
    ConstraintWeights_0=0.0904
    ConstraintWeights_1:float=0.0932
    ConstraintWeights_2:float=0
    ConstraintWeights_3:float=0.8013
    ConstraintWeights_4:float=0.0152
   ) -> (Pose) {
    alphaBoolBlend {blendOption=Linear}
    alphaScaleBiasClamp {bClampResult=true, bInterpResult=true, clampMax=0.5, interpSpeedDecreasing=20, interpSpeedIncreasing=20}
    boneToModify {boneName=dyn_belt_03}
    constraintSetup [
     {targetBone {boneName=pelvis}}
     {targetBone {boneName=thigh_l}}
     {targetBone {boneName=calf_l}}
     {targetBone {boneName=thigh_twist_01_l}}
     {targetBone {boneName=deform_groin_l}}
    ]
    constraintWeights [1,1,1,1,1]
   }
   n8 = "Constraint - dyn_belt_04 (PPPPP)"!(
    ComponentPose=@n7
    Alpha=1
    ConstraintWeights_0=0.0122
    ConstraintWeights_1:float=0.3838
    ConstraintWeights_2:float=0.0072
    ConstraintWeights_3:float=0.5968
    ConstraintWeights_4:float=0
   ) -> (Pose) {
    alphaBoolBlend {blendOption=Linear}
    alphaScaleBiasClamp {bClampResult=true, bInterpResult=true, clampMax=0.2, interpSpeedDecreasing=20, interpSpeedIncreasing=20}
    boneToModify {boneName=dyn_belt_04}
    constraintSetup [
     {targetBone {boneName=pelvis}}
     {targetBone {boneName=thigh_l}}
     {targetBone {boneName=calf_l}}
     {targetBone {boneName=thigh_twist_01_l}}
     {targetBone {boneName=deform_groin_l}}
    ]
    constraintWeights [1,1,1,1,1]
   }
   n9 = RigidBody!(ComponentPose=@n8, Alpha, OverrideWorldGravity:Vector="0,0,-980", ExternalForce) -> (Pose) {
    bClampLinearTranslationLimitToRefPose=true
    bOverrideWorldGravity=true
    overridePhysicsAsset=&/M_MED_SunTide_Upgrade_ShortsL_Physics
    simulationSpace=BaseBoneSpace
    alphaBoolBlend {blendOption=Linear}
    baseBoneRef {boneName=calf_l}
    componentAppliedLinearAccClamp=<1000 1000 1000>
    componentLinearAccScale=<0.2 0.2 0.2>
    componentLinearVelScale=<0.2 0.2 0.2>
    overrideWorldGravity {z=-980}
   }
   n10 = RigidBody(ComponentPose=@n9, OverrideWorldGravity="0,0,-980") {
    bClampLinearTranslationLimitToRefPose=true
    bOverrideWorldGravity=true
    overridePhysicsAsset=&/M_MED_SunTide_Upgrade_ShortsR_Physics
    simulationSpace=BaseBoneSpace
    alphaBoolBlend {blendOption=Linear}
    baseBoneRef {boneName=calf_r}
    componentAppliedLinearAccClamp=<1000 1000 1000>
    componentLinearAccScale=<0.2 0.2 0.2>
    componentLinearVelScale=<0.2 0.2 0.2>
    overrideWorldGravity {z=-980}
   }
   n11 = "Transform (Modify) Bone - Bone: dyn_chain_l_02"(ComponentPose=@n10, Alpha=0, Translation="-1.241,0.000,-1.000", Rotation="-10,0.000,31.917", Scale="1,1,1") {
    rotationMode=BMM_Additive
    translationMode=BMM_Additive
    alphaBoolBlend {blendOption=Linear}
    boneToModify {boneName=dyn_chain_l_02}
    rotation {pitch=-10, roll=31.917}
    translation {x=-1.241, z=-1}
   }
   n12 = Anim Dynamics - Bone: dyn_tassel_01_01(ComponentPose=@n11, Alpha=1, ExternalForce="0,0,0") {
    angularSpringConstant=100
    bAngularSpring=true
    bUseGravityOverride=true
    simulationSpace=RootRelative
    alphaBoolBlend {blendOption=Linear}
    boundBone {boneName=dyn_tassel_01_01}
    componentAppliedLinearAccClamp=<10000 10000 10000>
    componentLinearAccScale=<0.1 0.1 0.3>
    componentLinearVelScale=<0.1 0.1 0.3>
    gravityOverride {z=-980}
    physicsBodyDefinitions [{
      boundBone {boneName=dyn_tassel_01_01}
      boxExtents=<10 10 20>
      constraintSetup {angularLimitsMax=<0 70 70>, angularLimitsMin=<0 -70 -100>, angularTarget=<0 0 0>}
      localJointOffset=<1 0 0>
     }]
    relativeSpaceBone {boneName=lowerarm_twist_01_l}
   }
   n13 = Anim Dynamics - Bone: dyn_tassel_02_01(ComponentPose=@n12, Alpha=1, ExternalForce="0,0,0") {
    angularSpringConstant=100
    bAngularSpring=true
    bUseGravityOverride=true
    simulationSpace=RootRelative
    alphaBoolBlend {blendOption=Linear}
    boundBone {boneName=dyn_tassel_02_01}
    componentAppliedLinearAccClamp=<10000 10000 10000>
    componentLinearAccScale=<0.1 0.1 0.3>
    componentLinearVelScale=<0.1 0.1 0.3>
    gravityOverride {z=-980}
    physicsBodyDefinitions [{
      boundBone {boneName=dyn_tassel_02_01}
      boxExtents=<10 10 20>
      constraintSetup {angularLimitsMax=<0 70 70>, angularLimitsMin=<0 -70 -100>, angularTarget=<0 0 0>}
      localJointOffset=<1 0 0>
     }]
    relativeSpaceBone {boneName=lowerarm_twist_01_l}
   }
   n14 = Trail controller - Bone: dyn_tassel_01_04(ComponentPose=@n13, Alpha=1) {
    bUsePlanarLimit=true
    chainLength=3
    relaxationSpeedScale=4
    planarLimits [{drivingBone {boneName=lowerarm_twist_01_l}, planeTransform {location=<0 -3 0>, rotation=<rot 0 0 -110>, scale=<0.3 0.3 0.3>}}]
    rotationLimits [{limitMax=<180 180 180>}, {limitMax=<180 180 180>}, {limitMax=<180 180 180>}]
    rotationOffsets [<0 0 0>, <0 0 0>, <0 0 0>]
    trailBone {boneName=dyn_tassel_01_04}
   }
   n15 = Trail controller - Bone: dyn_tassel_02_04(ComponentPose=@n14, Alpha=1) {
    bUsePlanarLimit=true
    chainLength=3
    relaxationSpeedScale=3
    planarLimits [{drivingBone {boneName=lowerarm_twist_01_l}, planeTransform {location=<0 -3 0>, rotation=<rot 0 0 -110>, scale=<0.3 0.3 0.3>}}]
    rotationLimits [{limitMax=<180 180 180>}, {limitMax=<180 180 180>}, {limitMax=<180 180 180>}]
    rotationOffsets [<0 0 0>, <0 0 0>, <0 0 0>]
    trailBone {boneName=dyn_tassel_02_04}
   }
   n16 = Control Rig!(
    Source=Component To Local(ComponentPose=@n15)
    Alpha
    Crouch:bool=Get bIsCrouching()
    CrouchWalk:bool=Get Crouch_move() -> (Crouch_move:bool)
   ) -> (Pose) {
    alphaBoolBlend {blendOption=Linear}
    controlRigAssetReference {blueprintRigClass=&/M_MED_SunTide_Upgrade_CtrlRig.M_MED_SunTide_Upgrade_CtrlRig_C}
   }
   n17 = "Constraint - dyn_shorts_bk_l (P)"(ComponentPose=Local To Component(LocalPose=@n16), Alpha=0.8, ConstraintWeights_0=1) {
    alphaBoolBlend {blendOption=Linear}
    boneToModify {boneName=dyn_shorts_bk_l}
    constraintSetup [{targetBone {boneName=calf_l}}]
    constraintWeights [1]
   }
   n18 = "Constraint - dyn_shorts_out_l (P)"(ComponentPose=@n17, Alpha=0.3, ConstraintWeights_0=1) {
    alphaBoolBlend {blendOption=Linear}
    boneToModify {boneName=dyn_shorts_out_l}
    constraintSetup [{targetBone {boneName=dyn_shorts_bk_l}}]
    constraintWeights [1]
   }
   n19 = "Constraint - dyn_shorts_in_l (P)"(ComponentPose=@n18, Alpha=0.3, ConstraintWeights_0=1) {
    alphaBoolBlend {blendOption=Linear}
    boneToModify {boneName=dyn_shorts_in_l}
    constraintSetup [{targetBone {boneName=dyn_shorts_bk_l}}]
    constraintWeights [1]
   }
   n20 = "Constraint - dyn_shorts_bk_r (P)"(ComponentPose=@n19, Alpha=0.8, ConstraintWeights_0=1) {
    alphaBoolBlend {blendOption=Linear}
    boneToModify {boneName=dyn_shorts_bk_r}
    constraintSetup [{targetBone {boneName=calf_r}}]
    constraintWeights [1]
   }
   n21 = "Constraint - dyn_shorts_out_r (P)"(ComponentPose=@n20, Alpha=0.3, ConstraintWeights_0=1) {
    alphaBoolBlend {blendOption=Linear}
    boneToModify {boneName=dyn_shorts_out_r}
    constraintSetup [{targetBone {boneName=dyn_shorts_bk_r}}]
    constraintWeights [1]
   }
   n22 = "Constraint - dyn_shorts_in_r (P)"(ComponentPose=@n21, Alpha=0.3, ConstraintWeights_0=1) {
    alphaBoolBlend {blendOption=Linear}
    boneToModify {boneName=dyn_shorts_in_r}
    constraintSetup [{targetBone {boneName=dyn_shorts_bk_r}}]
    constraintWeights [1]
   }
   n23 = Control Rig!(Source=Component To Local(ComponentPose=@n22), Alpha) -> (Pose) {
    alphaBoolBlend {blendOption=Linear}
    controlRigAssetReference {blueprintRigClass=&/M_MED_SunTide_Upgrade_CtrlRig2.M_MED_SunTide_Upgrade_CtrlRig2_C}
   }
   Output Pose(Result=@n23) {layerGroup=None}
  ]}
 EventGraph {nodes [
   Event BlueprintUpdateAnimation()
   > n1 = Sequence!(execute) -> (then_0, then_1, then_2:Exec)
   > n2 = Gravity Override State Function DEPRECATED(CharacterPartAnimInstance=Self-Reference(), GameTimeInSeconds=GetGameTimeInSeconds(), GravityOverrideParams=Get Gravity Override Params())
   > Set gravity_override(gravity_override=@n2)
   n3 = Branch(execute=@n1.then_1, Condition=AND Boolean(A=Get bIsCrouching(), B=Get bIsMoving2D()))
   > Set Crouch_move(Crouch_move:bool=true) -> (Output_Get:bool)
   Set Crouch_move(execute=@n3.else, Crouch_move=false)
   Set IsCrouching(execute=@n1.then_2, IsCrouching:double=SelectFloat(A=1, B=0, bPickA=Get bIsCrouching())) -> (Output_Get:double)
   > Set Emote(Emote:double=SelectFloat(A=1, B=0, bPickA=Get bIsPlayingEmote())) -> (Output_Get:double)
   MakeRelativeTransform!(A:"Transform (by ref)", RelativeTo:"Transform (by ref)") -> (ReturnValue:Transform)
   BreakRotator(InRot="0, 0, 0")
   n4 = GetOwnerSkeletalMeshComponent()
   n5 = BreakTransform(InTransform=GetDeltaTransformFromRefPose(self=@n4, BoneName=dyn_chain_l_02, BaseName=None))
   n6 = BreakTransform(InTransform=GetDeltaTransformFromRefPose(self=@n4, BoneName=thigh_twist_01_l, BaseName=None))
   Set chain_l(chain_l:double=MapRangeClamped(Value="Distance (Vector)"!(V1:Vector=@n5, V2:Vector=@n6) -> (ReturnValue:double), InRangeA=-10, InRangeB=0)) -> (Output_Get:double)
  ]}
]
variables [
 chain_l {type=number}
 Crouch_move {type=boolean}
 Emote {type=number}
 Gravity Override Params {type=GravityOverrideParamsStruct}
 gravity_override {type=Vector}
 IsCrouching {type=number}
]
